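# Physics asset for the male Floatilla Captain pickaxe, providing secondary motion for its dangling parts.
# A root body of capsules and a box anchors a rope knot, seven teeth and a baseball bone, each constrained directly to it.
# Two rope chains, left and right, hang from the knot with wide swing and near-locked twist.
# The teeth wobble within tight limits, while the baseball swings widely on mostly one axis.
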PhysicsAsset /BRCosmetics/Weapons/FORT_Melee/Pickaxe_FloatillaCaptainMale/Meshes/Floatilla_Captain_Male_Axe_Physics
bodies {
 b_AR_Root [
  {shapeType=Capsule, length=13.4842, radius=2.8276, center=<0 0 -4.9878>, rotation=<rot 0 -90 0>}
  {shapeType=Capsule, length=4.7431, radius=0.5635, center=<0 0.7968 -18.2048>, rotation=<rot 0 0 -31.869>}
  {shapeType=Capsule, length=5.6928, radius=1.1199, center=<0 2.7328 -15.9376>, rotation=<rot 0 0 5.0001>}
  {shapeType=Box, extentX=24.5, extentY=22, extentZ=9.5, center=<0 0 14.7433>}
 ]
 dyn_baseball_MSTR [{shapeType=Sphere, radius=4.505, center=<13.6334 0 3.4335>}]
 dyn_rope_knot [{shapeType=Sphere, radius=2.2712, center=<-0.0702 -0.1416 -0.0668>}]
 dyn_rope_l_01 [{shapeType=Capsule, length=0.4541, radius=0.5559, center=<1.0114 -0.0218 0.048>, rotation=<rot 90 35.2644 35.2643>}]
 dyn_rope_l_01_b [{shapeType=Capsule, length=0.5327, radius=0.7458, center=<0.8589 0.0005 -0.0175>, rotation=<rot 90 0 0>}]
 dyn_rope_l_02 [{shapeType=Capsule, length=2.5779, radius=1.0221, center=<2.3513 -0.0001 0>, rotation=<rot -90 -35.2644 35.2643>}]
 dyn_rope_l_03 [{shapeType=Capsule, length=1.3613, radius=1.0611, center=<1.331 0 0>, rotation=<rot 90 0.0149 0.0149>}]
 dyn_rope_r_01 [{shapeType=Capsule, length=1.1603, radius=0.7318, center=<1.3669 -0.1195 0.1729>, rotation=<rot 90 -0.0033 -0.0032>}]
 dyn_rope_r_02 [{shapeType=Capsule, length=2.8427, radius=1.0589, center=<1.9216 -0.0009 -0.0553>, rotation=<rot 90 0 0>}]
 dyn_rope_r_03 [{shapeType=Capsule, length=2.5216, radius=1.2774, center=<1.7435 0 0>, rotation=<rot 90 0 0>}]
 dyn_rope_r_04 [{shapeType=Capsule, length=0.9052, radius=1.2209, center=<0.6173 0 0>, rotation=<rot -90 0 0>}]
 dyn_teeth_01 [{shapeType=Box, extentX=1.1794, extentY=2.5001, extentZ=3.8715, center=<0.2339 -0.0012 -0.2427>, rotation=<rot 0 -90.0414 -65.5685>}]
 dyn_teeth_02 [{shapeType=Box, extentX=1.9056, extentY=3.4345, extentZ=5.441, center=<0.4368 -0.0009 -0.3692>, rotation=<rot 0 -90.0228 -67.1926>}]
 dyn_teeth_03 [{shapeType=Box, extentX=1.8851, extentY=4.3364, extentZ=6.8564, center=<0.5606 -0.0013 -0.4741>, rotation=<rot 0 -90.0177 -68.1155>}]
 dyn_teeth_04 [{shapeType=Box, extentX=2.2259, extentY=5.7519, extentZ=8.9369, center=<0.4883 -0.0011 -0.5463>, rotation=<rot 0 -90.014 -64.2633>}]
 dyn_teeth_05 [{shapeType=Box, extentX=2.4518, extentY=6.2582, extentZ=10.2707, center=<0.9064 -0.0012 -0.7376>, rotation=<rot 0 -90.0072 -72.4837>}]
 dyn_teeth_06 [{shapeType=Box, extentX=2.6897, extentY=6.7798, extentZ=10.6865, center=<1.0056 -0.0012 -0.8197>, rotation=<rot 0 -90.0052 -72.5563>}]
 dyn_teeth_07 [{shapeType=Box, extentX=2.6897, extentY=7.0348, extentZ=10.8617, center=<1.3422 -0.0012 -0.9114>, rotation=<rot 0 -90.0055 -74.8582>}]
}
constraints [
 {bone1Name=dyn_rope_knot, bone2Name=b_AR_Root, swing1LimitDegrees=15, swing2LimitDegrees=15, twistLimitDegrees=1}
 {bone1Name=dyn_rope_r_01, bone2Name=dyn_rope_knot, swing1LimitDegrees=60, swing2LimitDegrees=60, twistLimitDegrees=1}
 {bone1Name=dyn_rope_r_02, bone2Name=dyn_rope_r_01, swing1LimitDegrees=60, swing2LimitDegrees=60, twistLimitDegrees=1}
 {bone1Name=dyn_rope_r_03, bone2Name=dyn_rope_r_02, swing1LimitDegrees=45, swing2LimitDegrees=45, twistLimitDegrees=1}
 {bone1Name=dyn_rope_r_04, bone2Name=dyn_rope_r_03, swing1LimitDegrees=45, swing2LimitDegrees=45, twistLimitDegrees=1}
 {bone1Name=dyn_teeth_07, bone2Name=b_AR_Root, swing1LimitDegrees=10, swing2LimitDegrees=5, twistLimitDegrees=1}
 {bone1Name=dyn_teeth_06, bone2Name=b_AR_Root, swing1LimitDegrees=10, swing2LimitDegrees=5, twistLimitDegrees=1}
 {bone1Name=dyn_teeth_05, bone2Name=b_AR_Root, swing1LimitDegrees=10, swing2LimitDegrees=5, twistLimitDegrees=1}
 {bone1Name=dyn_teeth_04, bone2Name=b_AR_Root, swing1LimitDegrees=10, swing2LimitDegrees=5, twistLimitDegrees=1}
 {bone1Name=dyn_teeth_03, bone2Name=b_AR_Root, swing1LimitDegrees=10, swing2LimitDegrees=5, twistLimitDegrees=1}
 {bone1Name=dyn_teeth_02, bone2Name=b_AR_Root, swing1LimitDegrees=10, swing2LimitDegrees=5, twistLimitDegrees=1}
 {bone1Name=dyn_teeth_01, bone2Name=b_AR_Root, swing1LimitDegrees=10, swing2LimitDegrees=5, twistLimitDegrees=1}
 {bone1Name=dyn_baseball_MSTR, bone2Name=b_AR_Root, swing1LimitDegrees=2, swing2LimitDegrees=85, twistLimitDegrees=1}
 {bone1Name=dyn_rope_l_01, bone2Name=dyn_rope_knot, swing1LimitDegrees=70, swing2LimitDegrees=70, twistLimitDegrees=1}
 {bone1Name=dyn_rope_l_01_b, bone2Name=dyn_rope_l_01, swing1LimitDegrees=70, swing2LimitDegrees=70, twistLimitDegrees=1}
 {bone1Name=dyn_rope_l_02, bone2Name=dyn_rope_l_01_b, swing1LimitDegrees=45, swing2LimitDegrees=45, twistLimitDegrees=1}
 {bone1Name=dyn_rope_l_03, bone2Name=dyn_rope_l_02, swing1LimitDegrees=45, swing2LimitDegrees=45, twistLimitDegrees=0}
]
